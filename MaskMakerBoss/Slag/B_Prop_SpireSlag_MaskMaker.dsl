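; Slag shard prop spawned by the Mask Maker boss that telegraphs, then erupts as an area attack.
; On the server it waits a delay, plays a burst cue and sphere-traces twice, applying damage and knockback effects, then environmental damage that suppresses loot from other slag props.
; It then snaps to a randomized end scale through replicated ShardScale; the eased ScaleSlagShard event is never started here.
; EndPlay decrements the boss's shard count.
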(event EventBeginPlay
  (|CallParentFunction)
  (Networking|SwitchHasAuthority
    (:Authority
      (CallFunction|SetLootDropsEnabled true)
      (bind _asbp_player_pawn_cherry_smoke_mask_maker_boss (Utilities|Casting|CastToBP_PlayerPawn_CherrySmoke_MaskMakerBoss (Game|GetInstigator)))
      (bind _output_get (Variables|Default|SetMaskMakerPawn _asbp_player_pawn_cherry_smoke_mask_maker_boss))
      (CallFunction|AddLootBlockingKiller _output_get)
      (Default|BindEventtoSpireShardAbilityStarted _output_get (EventDispatchers|CreateEvent :!event Custom|NewSlagShardsSpawning))
      (bind _output_get_2 (Variables|Default|SetStartShardScale (Math|Vector|GetMaxElement (Transformation|GetActorScale3D))))
      (Variables|Default|SetShardScale (Math|Conversions|ToVector(Float) _output_get_2))
      (CallFunction|ScaleTelegraph)
      (Utilities|FlowControl|Delay (ScalableFloat|GetValueatLevel (Variables|Hotfix|GetScaleShardDelay_Hf)))
      (CallFunction|ScaleStarted)
      (Variables|Default|SetShardScale (Math|Conversions|ToVector(Float) (+ (ScalableFloat|GetValueatLevel (Variables|Hotfix|GetShardEndScale_Hf)) (Math|Random|RandomFloatinRange (* (ScalableFloat|GetValueatLevel (Variables|Hotfix|GetShardEndScaleVar_Hf)) -1.0) (ScalableFloat|GetValueatLevel (Variables|Hotfix|GetShardEndScaleVar_Hf))))))
      (CallFunction|ScaleFinished))))

(event Custom|ScaleSlagShard
  (Variables|Default|SetScaleShardAlpha (+ (Variables|Default|GetScaleShardAlpha) (/ (ScalableFloat|GetValueatLevel (Variables|Hotfix|GetScaleShardPeriod_Hf)) (ScalableFloat|GetValueatLevel (Variables|Hotfix|GetScaleShardDuration_Hf)))))
  (Variables|Default|SetShardScale (Math|Conversions|ToVector(Float) (Math|Interpolation|Ease "EaseIn" (Variables|Default|GetScaleShardAlpha) (Variables|Default|GetStartShardScale) 1.0 2.0)))
  (if (>= (Variables|Default|GetScaleShardAlpha) 1.0)
    (Utilities|Time|ClearandInvalidateTimerbyHandle (Variables|Default|GetScaleShardTHandle))
    (CallFunction|ScaleFinished)))

(event OnLinkedActorConnected(FortLinkToActor)
  (params (LinkedActor :type Actor%20Object%20Reference))
  (Variables|Default|SetShardLinkedActor LinkedActor))

(event OnLinkedActorDestroyed(FortLinkToActor)
  (params (DamageCauser :type Actor%20Object%20Reference))
  (Utilities|Casting|CastToBP_PlayerPawn_CherrySmoke_MaskMakerBoss (Game|GetInstigator DamageCauser))
  (CallFunction|SetLootDropsEnabled false))

(event Custom|ScaleTelegraph
  (GameplayCue|ExecuteGameplayCueOnActor(Burst) self (Variables|GameplayCue|GetGC_SpireShardTelegraph) (Ability|GameplayCue|MakeGameplayCueParameters 0.0)))

(event Custom|ScaleFinished)

(event Custom|ScaleStarted
  (GameplayCue|ExecuteGameplayCueOnActor(Burst) self (Variables|GameplayCue|GetGC_SpireShardBurst) (Ability|GameplayCue|MakeGameplayCueParameters 0.0))
  (bind _outhits (Collision|MultiSphereTraceByChannel (Transformation|GetActorLocation) (+ (Transformation|GetActorLocation) "0.000000,0.000000,1.000000") (ScalableFloat|GetValueatLevel (Variables|Hotfix|GetShardAoeRadius_Hf)) "TraceTypeQuery9" false (Utilities|Array|MakeArray (Variables|Default|GetMaskMakerPawn) (Utilities|Array|Get(acopy) (Variables|/MaskMakerBoss/Pawn/BP_PlayerPawn_CherrySmoke_MaskMakerBoss.BP_PlayerPawn_CherrySmoke_MaskMakerBoss_C|GetMaskWeakpoints (Variables|Default|GetMaskMakerPawn))) (Utilities|Array|Get(acopy) (Variables|/MaskMakerBoss/Pawn/BP_PlayerPawn_CherrySmoke_MaskMakerBoss.BP_PlayerPawn_CherrySmoke_MaskMakerBoss_C|GetMaskWeakpoints (Variables|Default|GetMaskMakerPawn)) 1) (Utilities|Array|Get(acopy) (Variables|/MaskMakerBoss/Pawn/BP_PlayerPawn_CherrySmoke_MaskMakerBoss.BP_PlayerPawn_CherrySmoke_MaskMakerBoss_C|GetMaskWeakpoints (Variables|Default|GetMaskMakerPawn)) 2) (Utilities|Array|Get(acopy) (Variables|/MaskMakerBoss/Pawn/BP_PlayerPawn_CherrySmoke_MaskMakerBoss.BP_PlayerPawn_CherrySmoke_MaskMakerBoss_C|GetMaskWeakpoints (Variables|Default|GetMaskMakerPawn)) 3)) "None" true "(R=1.000000,G=0.000000,B=0.000000,A=1.000000)" "(R=0.000000,G=1.000000,B=0.000000,A=1.000000)" 2.0))
  (for _array_element _outhits
    (Utilities|IsValid (.HitActor (Collision|BreakHitResult _array_element))
      (:Is%20Valid
        (Fortnite|ApplyGameplayEffectsToActor (.HitActor (Collision|BreakHitResult _array_element)) (Utilities|Array|MakeArray (Variables|Default|GetDamageGE)) (Transformation|GetActorLocation (Variables|Default|GetMaskMakerPawn)) false 0 (Ability|GetAbilitySystemComponent (Variables|Default|GetMaskMakerPawn)))
        (Fortnite|ApplyGameplayEffectsToActor (.HitActor (Collision|BreakHitResult _array_element)) (Utilities|Array|MakeArray (Variables|Default|GetKnockbackGE)) (Transformation|GetActorLocation) false 0 (Ability|GetAbilitySystemComponent 0)))))
  (bind _outhits_2 (Collision|MultiSphereTraceByChannel (Transformation|GetActorLocation) (+ (Transformation|GetActorLocation) "0.000000,0.000000,1.000000") (ScalableFloat|GetValueatLevel (Variables|Hotfix|GetShardAoeRadius_Hf)) "TraceTypeQuery10" false (Utilities|Array|MakeArray (Utilities|Array|Get(acopy) (Variables|/MaskMakerBoss/Pawn/BP_PlayerPawn_CherrySmoke_MaskMakerBoss.BP_PlayerPawn_CherrySmoke_MaskMakerBoss_C|GetMaskWeakpoints (Variables|Default|GetMaskMakerPawn))) (Utilities|Array|Get(acopy) (Variables|/MaskMakerBoss/Pawn/BP_PlayerPawn_CherrySmoke_MaskMakerBoss.BP_PlayerPawn_CherrySmoke_MaskMakerBoss_C|GetMaskWeakpoints (Variables|Default|GetMaskMakerPawn)) 1) (Utilities|Array|Get(acopy) (Variables|/MaskMakerBoss/Pawn/BP_PlayerPawn_CherrySmoke_MaskMakerBoss.BP_PlayerPawn_CherrySmoke_MaskMakerBoss_C|GetMaskWeakpoints (Variables|Default|GetMaskMakerPawn)) 2) (Utilities|Array|Get(acopy) (Variables|/MaskMakerBoss/Pawn/BP_PlayerPawn_CherrySmoke_MaskMakerBoss.BP_PlayerPawn_CherrySmoke_MaskMakerBoss_C|GetMaskWeakpoints (Variables|Default|GetMaskMakerPawn)) 3) (Variables|Default|GetShardLinkedActor)) "None" true "(R=1.000000,G=0.000000,B=0.000000,A=1.000000)" "(R=0.000000,G=1.000000,B=0.000000,A=1.000000)" 2.0))
  (for _array_element_2 _outhits_2
    (Utilities|IsValid (.HitActor (Collision|BreakHitResult _array_element_2))
      (:Is%20Valid
        (bind _asb_prop_spire_slag_parent (Utilities|Casting|CastToB_Prop_SpireSlag_Parent (.HitActor (Collision|BreakHitResult _array_element_2)))
          (:then
            (CallFunction|SetLootDropsEnabled _asb_prop_spire_slag_parent false))
          (:CastFailed
            _))
        (Game|Damage|ApplyDamage (.HitActor (Collision|BreakHitResult _array_element_2)) (ScalableFloat|GetValueatLevel (Variables|Hotfix|GetShardEnvDmg_Hf)) (Pawn|GetController (Variables|Default|GetMaskMakerPawn)) (Variables|Default|GetMaskMakerPawn) "/Engine/EngineDamageTypes/DmgTypeBP_Environmental.DmgTypeBP_Environmental_C")))))

(event Custom|NewSlagShardsSpawning)

(event EventOnDamagePlayEffects
  (params (Damage :type Float%20%28single-precision%29) (DamageTags :type Gameplay%20Tag%20Container%20Structure%20%28by%20ref%29) (Momentum :type Vector :default "0, 0, 0") (HitInfo :type Hit%20Result%20Structure%20%28by%20ref%29) (InstigatedBy :type Fort%20Pawn%20Object%20Reference) (DamageCauser :type Actor%20Object%20Reference) (EffectContext :type Gameplay%20Effect%20Context%20Handle%20Structure))
  (|CallParentFunction Damage DamageTags Momentum HitInfo InstigatedBy DamageCauser EffectContext))

(event EventEndPlay
  (params (EndPlayReason :type EEndPlayReason%20Enum :default "Destroyed"))
  (bind _maskmakerpawn (Variables|Default|GetMaskMakerPawn))
  (|CallParentFunction EndPlayReason)
  (Utilities|IsValid _maskmakerpawn
    (:Is%20Valid
      (Math|Integer|DecrementInt (Variables|/MaskMakerBoss/Pawn/BP_PlayerPawn_CherrySmoke_MaskMakerBoss.BP_PlayerPawn_CherrySmoke_MaskMakerBoss_C|GetTotalShardSpawnedAmount _maskmakerpawn)))))

(fn OnRep_ShardScale
  (Transformation|SetActorScale3D (Variables|Default|GetShardScale)))

(fn SetOrganicTransform
  (params (ScaleRangeMult :type Vector%202D%20Structure) (RandomYawAdd :type Vector%202D%20Structure) (RandomPitchAdd :type Vector%202D%20Structure) (RandomZAdd :type Vector%202D%20Structure)))

(fn ConstructionScript
  (|CallParentFunction))
